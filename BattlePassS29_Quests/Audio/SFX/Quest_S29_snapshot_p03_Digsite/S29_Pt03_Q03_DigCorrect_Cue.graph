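# Sound cue for the correct-dig feedback in a Season 29 Battle Pass digsite snapshot quest, part 3.
# A Random node picks one of three equally weighted DigCorrect wave variations on each play.
# The cue uses the shared Season 29 quest attenuation asset and the Fort_Quests sound class.
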SoundCue /BattlePassS29_Quests/Audio/SFX/Quest_S29_snapshot_p03_Digsite/S29_Pt03_Q03_DigCorrect_Cue
root=Random_0
nodes [
 {id=Random_0, children [WavePlayer_3,WavePlayer_4,WavePlayer_5], properties {weights [1,1,1]}}
 {id=WavePlayer_3, properties {soundWaveAssetPtr=&/S29_Pt03_Q03_DigCorrect_01}}
 {id=WavePlayer_4, properties {soundWaveAssetPtr=&/S29_Pt03_Q03_DigCorrect_02}}
 {id=WavePlayer_5, properties {soundWaveAssetPtr=&/S29_Pt03_Q03_DigCorrect_03}}
]
properties {
 attenuationSettings=&/S29_Quests_Atten
 maxDistance=1300
 soundClassObject=&Fort_Quests
 volumeMultiplier=1
}
